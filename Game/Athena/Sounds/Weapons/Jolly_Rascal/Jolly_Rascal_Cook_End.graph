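# Sound cue for the end of the Jolly Rascal weapon's cook action, played as player foley.
# A Random node picks one of four equally weighted cook-end wave variations and feeds it through a Modulator with pitch variation disabled.
# It plays at reduced volume with weapon reload attenuation, and a concurrency override prevents new instances when the limit is reached.
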SoundCue /Game/Athena/Sounds/Weapons/Jolly_Rascal/Jolly_Rascal_Cook_End
root=Modulator_1
nodes [
 {id=Modulator_1, children [Random_0], properties {pitchMax=1, pitchMin=1}}
 {id=Random_0, children [WavePlayer_0,WavePlayer_2,WavePlayer_3,WavePlayer_4], properties {weights [1,1,1,1]}}
 {id=WavePlayer_0, properties {soundWaveAssetPtr=&/WPN_Jolly_Rascal_Cook_End_04}}
 {id=WavePlayer_2, properties {soundWaveAssetPtr=&/WPN_Jolly_Rascal_Cook_End_03}}
 {id=WavePlayer_3, properties {soundWaveAssetPtr=&/WPN_Jolly_Rascal_Cook_End_02}}
 {id=WavePlayer_4, properties {soundWaveAssetPtr=&/WPN_Jolly_Rascal_Cook_End_01}}
]
properties {
 attenuationSettings=&Fort_Weapon_Reload_Att
 bOverrideConcurrency=true
 maxDistance=2250
 soundClassObject=&Fort_Player_Foley
 volumeMultiplier=0.4
 concurrencyOverrides {resolutionRule=PreventNew}
}
