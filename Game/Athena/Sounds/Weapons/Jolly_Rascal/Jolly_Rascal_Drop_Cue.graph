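# Sound cue for dropping the Jolly Rascal weapon as loot, classed under the loot collect sound class.
# A single wave player plays one drop sample with a boosted volume and the shared weapon drop attenuation.
# An overridden concurrency limit caps simultaneous instances at a few and prevents new ones from starting beyond that.
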SoundCue /Game/Athena/Sounds/Weapons/Jolly_Rascal/Jolly_Rascal_Drop_Cue
root=WavePlayer_0
nodes [{id=WavePlayer_0, properties {soundWaveAssetPtr=&/WPN_Jolly_Rascal_Drop_01}}]
properties {
 attenuationSettings=&Fort_Weapon_Drop_Att
 bOverrideConcurrency=true
 maxDistance=3000
 soundClassObject=&Fort_Loot_Collect
 volumeMultiplier=1.7231
 concurrencyOverrides {maxCount=3, resolutionRule=PreventNew}
}
