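# Sound cue for the Jolly Rascal weapon's equip sound in Athena, classed as player foley.
# A single wave player feeds a modulator with pitch variation flattened to a constant.
# It uses close-range prop attenuation and overrides concurrency to prevent new instances from starting once the limit is reached.
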SoundCue /Game/Athena/Sounds/Weapons/Jolly_Rascal/Jolly_Rascal_Equip_Cue
root=Modulator_1
nodes [
 {id=Modulator_1, children [WavePlayer_1], properties {pitchMax=1, pitchMin=1}}
 {id=WavePlayer_1, properties {soundWaveAssetPtr=&/WPN_Jolly_Rascal_Equip_01}}
]
properties {
 attenuationSettings=/Game/Sounds/Attenuation/Fort_Prop_Close_Att
 bOverrideConcurrency=true
 maxDistance=1800
 soundClassObject=&Fort_Player_Foley
 volumeMultiplier=1
 concurrencyOverrides {resolutionRule=PreventNew}
}
