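# Looping fuse sound cue for the Jolly Rascal weapon's projectile, classed under explosives.
# A single fuse loop wave runs through an Enveloper that ramps volume and pitch upward over two seconds, then through a Doppler node with raised intensity.
# It uses overridden logarithmic attenuation and a concurrency override that blocks new instances once the limit is reached.
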SoundCue /Game/Athena/Sounds/Weapons/Jolly_Rascal/Jolly_Rascal_Fuse_Loop_Cue
root=Doppler_0
nodes [
 {id=Doppler_0, children [Enveloper_0], properties {dopplerIntensity=2}}
 {
  id=Enveloper_0
  children [WavePlayer_1]
  properties {pitchCurve {editorCurveData {keys [{time=0}, <at 2 1.0171>]}}, volumeCurve {editorCurveData {keys [<at 0 0.1>,<at 2 2.0153>]}}}
 }
 {id=WavePlayer_1, properties {soundWaveAssetPtr=&/WPN_Jolly_Rascal_Proj_Fuse_Loop_01}}
]
properties {
 attenuationSettings=/Game/Sounds/Attenuation/Fort_Explosive_Fuse_Ambient_Att
 bOverrideAttenuation=true
 bOverrideConcurrency=true
 maxDistance=1700
 soundClassObject=/Game/Sounds/FortSoundClassesAndModes/Fort_Explosives
 volumeMultiplier=0.5
 attenuationOverrides {distanceAlgorithm=Logarithmic, falloffDistance=1200, attenuationShapeExtents {x=500}}
 concurrencyOverrides {resolutionRule=PreventNew}
}
