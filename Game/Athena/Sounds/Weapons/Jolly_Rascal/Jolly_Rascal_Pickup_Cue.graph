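# Sound cue for picking up the Jolly Rascal weapon in Athena.
# A single wave player feeds the pickup recording straight to the output, with no randomization or modulation.
# It uses the shared weapon pickup attenuation and loot collect sound class, with boosted volume.
# Its concurrency override prevents new instances from starting once the limit is reached, rather than cutting off those already playing.
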SoundCue /Game/Athena/Sounds/Weapons/Jolly_Rascal/Jolly_Rascal_Pickup_Cue
root=WavePlayer_0
nodes [{id=WavePlayer_0, properties {soundWaveAssetPtr=&/WPN_Jolly_Rascal_Pickup_01}}]
properties {
 attenuationSettings=&Fort_Weapon_Pickup_Att
 bOverrideConcurrency=true
 maxDistance=3500
 soundClassObject=&Fort_Loot_Collect
 volumeMultiplier=1.7231
 concurrencyOverrides {resolutionRule=PreventNew}
}
